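; BP_LightFunction_POI_Mask is an actor whose cylinder appears to mark point-of-interest areas for the dynamic light function.
; Its construction script scales the cylinder uniformly, then branches on Update.
; When true, it calls Update RT, which refreshes every GenerateDynamicLightFunction actor's render target, then shows the cylinder.
; When false, it sets Update true and hides the cylinder; the BeginPlay, Tick and overlap events are empty.
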(event EventBeginPlay)

(event Collision|EventActorBeginOverlap
  (params (OtherActor :type Actor%20Object%20Reference)))

(event EventTick
  (params (DeltaSeconds :type Float%20%28single-precision%29)))

(event Custom|Update%20RT
  (bind _outactors (Actor|GetAllActorsOfClass 0 "/Game/ArtTools/Tech_Art/DynamicLightFunction/GenerateDynamicLightFunction.GenerateDynamicLightFunction_C"))
  (for _array_element _outactors
    (Class|GenerateDynamicLightFunction|1UpdateRT _array_element)))

(fn setMeshVis
  (Rendering|SetVisibility (Variables|Default|GetCylinder) true false))

(fn ConstructionScript
  (Transformation|SetRelativeScale3D (Variables|Default|GetCylinder) "20.000000,20.000000,20.000000")
  (if (Variables|Default|GetUpdate)
    (CallFunction|UpdateRT)
    (Rendering|SetVisibility (Variables|Default|GetCylinder) true false)
    (else
      (Variables|Default|SetUpdate true)
      (Rendering|SetVisibility (Variables|Default|GetCylinder) false false))))
