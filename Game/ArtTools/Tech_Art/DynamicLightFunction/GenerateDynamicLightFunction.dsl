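; Editor tool Blueprint that generates a world-space light function texture from BP_LightFunction_POI_Mask actors.
; Construction scales a preview plane to WorldSize, creates a render target and feeds it, with WorldSize and offset, to a preview material.
; The update clears the target and draws a brush material per mask actor, using its relative position, scale and Hardness.
; A separate bake event converts the render target into OutputTexture, editor-only.
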(event EventBeginPlay)

(event Collision|EventActorBeginOverlap
  (params (OtherActor :type Actor%20Object%20Reference)))

(event EventTick
  (params (DeltaSeconds :type Float%20%28single-precision%29)))

(event Custom|SetupRenderTarget
  (bind _returnvalue (Rendering|CreateRenderTarget2D 0 1024 1024 "RTF_RGBA8"))
  (Variables|Default|SetRenderTarget _returnvalue)
  (bind _returnvalue_2 (Rendering|Material|CreateDynamicMaterialInstance 0 "/Game/ArtTools/Tech_Art/DynamicLightFunction/M_Lightfunction_Preview.M_Lightfunction_Preview" "None" "None" :!class /Script/Engine.KismetMaterialLibrary))
  (Variables|Default|SetPreviewMID _returnvalue_2)
  (CallFunction|UpdatePreviewMateral))

(event Custom|Update%20PreviewMateral
  (bind _returnvalue (Transformation|GetActorLocation))
  (bind _returnvalue_2 (* _returnvalue -1.0))
  (bind _x (.x _returnvalue_2))
  (bind _y (.y _returnvalue_2))
  (bind _previewmid (Variables|Default|GetPreviewMID))
  (bind _worldsize (Variables|Default|GetWorldSize))
  (Rendering|Material|SetTextureParameterValue (Variables|Default|GetPreviewMID) "RenderTarget" (Variables|Default|GetRenderTarget))
  (Rendering|Material|SetScalarParameterValue _previewmid "WorldSize" _worldsize :!class /Script/Engine.MaterialInstanceDynamic)
  (Rendering|Material|SetVectorParameterValue _previewmid "Offset" (Utilities|Struct|MakeLinearColor _x _y) :!class /Script/Engine.MaterialInstanceDynamic)
  (Rendering|Material|SetMaterial (Variables|Default|GetPlane) 0 (Variables|Default|GetPreviewMID))
  (Utilities|Platform|CollectGarbage)
  (Development|PrintString 0 (Utilities|String|Append "LFG Lightfunction Material  Offset = " (Utilities|String|Append (Utilities|String|Append (Utilities|String|ToString(Float) _x) " / ") (Utilities|String|ToString(Float) _y))))
  (Development|PrintString 0 (Utilities|String|Append "LFG Lightfunction Material  Size = " (Utilities|String|ToString(Float) _worldsize))))

(event Custom|1_UpdateRT
  (bind _worldsize (Variables|Default|GetWorldSize))
  (Utilities|FlowControl|Sequence
    (:then_0
      (bind _outactors (Actor|GetAllActorsOfClass 0 "/Game/ArtTools/Tech_Art/DynamicLightFunction/BP_LightFunction_POI_Mask.BP_LightFunction_POI_Mask_C"))
      (for _array_element _outactors
        (Class|BPLightFunctionPOIMask|SetMeshVis _array_element)))
    (:then_1
      (CallFunction|UpdatePreviewMateral)
      (Rendering|ClearRenderTarget2D 0 (Variables|Default|GetRenderTarget))
      (bind _outactors_2 (Actor|GetAllActorsOfClass 0 "/Game/ArtTools/Tech_Art/DynamicLightFunction/BP_LightFunction_POI_Mask.BP_LightFunction_POI_Mask_C"))
      (for _array_element_2 _outactors_2
        (bind _returnvalue (Rendering|Material|CreateDynamicMaterialInstance 0 "/Game/ArtTools/Tech_Art/DynamicLightFunction/M_LightFunctionMask_Brush.M_LightFunctionMask_Brush" "None" "None" :!class /Script/Engine.KismetMaterialLibrary))
        (bind _output_get (Variables|Default|SetBrush_MID _returnvalue))
        (Rendering|Material|SetScalarParameterValue _output_get "Size" (/ (* (.x (Transformation|GetActorScale3D _array_element_2)) 2e+03) _worldsize) :!class /Script/Engine.MaterialInstanceDynamic)
        (Rendering|Material|SetVectorParameterValue (Variables|Default|GetBrush_MID) "PosOffset" (Math|Color|MakeColor (.x (* (/ (- (Transformation|GetActorLocation _array_element_2) (Transformation|GetActorLocation)) (* (Variables|Default|GetWorldSize) (/ (* (.x (Transformation|GetActorScale3D _array_element_2)) 2e+03) _worldsize))) -1.0)) (.y (* (/ (- (Transformation|GetActorLocation _array_element_2) (Transformation|GetActorLocation)) (* (Variables|Default|GetWorldSize) (/ (* (.x (Transformation|GetActorScale3D _array_element_2)) 2e+03) _worldsize))) -1.0))) :!class /Script/Engine.MaterialInstanceDynamic)
        (Rendering|Material|SetScalarParameterValue (Variables|Default|GetBrush_MID) "Hardness" (Variables|/Game/ArtTools/Tech_Art/DynamicLightFunction/BP_LightFunction_POI_Mask.BP_LightFunction_POI_Mask_C|GetHardness _array_element_2) :!class /Script/Engine.MaterialInstanceDynamic)
        (Rendering|DrawMaterialtoRenderTarget 0 (Variables|Default|GetRenderTarget) (Variables|Default|GetBrush_MID))))))

(event Custom|2_Bake%20Texture
  (EditorScripting|Rendering|ConvertRenderTargettoTexture2DEditorOnly 0 (Variables|Default|GetRenderTarget) (Variables|Default|GetOutputTexture)))

(fn ConstructionScript
  (Transformation|SetWorldScale3D (Variables|Default|GetPlane) (Math|Vector|MakeVector (/ (Variables|Default|GetWorldSize) 1e+02) (/ (Variables|Default|GetWorldSize) 1e+02) 1.0))
  (CallFunction|SetupRenderTarget)
  (if (Variables|Default|GetUpdate)
    (CallFunction|1UpdateRT)
    (else
      (Variables|Default|SetUpdate true))))
