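# Post-process material with additive blend that, by its name, serves as a brush for a dynamic light function mask.
# It scales the UVs about their center by a uniform Size parameter, then shifts them by the XY of PosOffset.
# A SphereMask compares the result to the UV center at a fixed radius, giving a movable, resizable circle whose edge falloff follows Hardness.
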Material /Game/ArtTools/Tech_Art/DynamicLightFunction/M_LightFunctionMask_Brush
graph {
 properties {
  blendMode=BLEND_Additive
  blendableOutputAlpha=true
  materialDomain=MD_PostProcess
  pixelDepthOffsetMode=PDOM_Legacy
  shadingModel=MSM_Unlit
  usedShadingModels=Unlit
 }
 nodes [
  Constant() {r=1}
  n1 = TextureCoordinate()
  n2 = ScalarParameter() {defaultValue=1, parameterName=Size}
  n3 = AppendVector(A=@n2, B=@n2)
  n4 = VectorParameter() {parameterName=PosOffset, defaultValue=<rgba 0 0 0 1>}
  SphereMask(
   A=Add(
    A=MaterialFunctionCall(UVs=@n1, Texture Scale=@n3) {materialFunction=&ScaleUVsByCenter}.UVs
    B=ComponentMask(None=@n4.RGB) {g=true, r=true}
   )
   B=Constant2Vector() {g=0.5, r=0.5}
   Radius=Constant() {r=0.5}
   Hardness=ScalarParameter() {parameterName=Hardness}
  )
 ]
}
