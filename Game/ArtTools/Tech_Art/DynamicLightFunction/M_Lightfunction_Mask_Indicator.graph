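# Translucent, unlit, two-sided material that, by its name, serves as a visual indicator for a dynamic light function mask.
# A checker grid texture is multiplied by a color parameter that defaults to red, giving a tinted grid pattern.
# A separate scalar Opacity parameter defaults to partial transparency.
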Material /Game/ArtTools/Tech_Art/DynamicLightFunction/M_Lightfunction_Mask_Indicator
graph {
 properties {blendMode=BLEND_Translucent, pixelDepthOffsetMode=PDOM_Legacy, shadingModel=MSM_Unlit, twoSided=true, usedShadingModels=Unlit}
 nodes [
  Multiply(A=TextureSample() {texture=/Game/Maps/Textures/checkerGrid}.RGB, B=VectorParameter() {defaultValue=<rgba 1 0 0 1>}.RGB)
  ScalarParameter() {defaultValue=0.6, parameterName=Opacity}
 ]
}
