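# Unlit translucent material that, by its name, previews a dynamic light function, though none of its nodes connects to a material output.
# The main chain projects a RenderTarget texture across world XY using Offset and WorldSize parameters centered on UV 0.5, then clamps and halves the sample.
# Separate fragments hold a Strips-driven UV sign test, a UV rotator and an unused Color parameter.
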Material /Game/ArtTools/Tech_Art/DynamicLightFunction/M_Lightfunction_Preview
graph {
 properties {blendMode=BLEND_Translucent, pixelDepthOffsetMode=PDOM_Legacy, shadingModel=MSM_Unlit, usedShadingModels=Unlit}
 nodes [
  n1 = TextureCoordinate()
  n2 = ComponentMask(None=@n1) {r=true}
  n3 = ScalarParameter() {defaultValue=9.7188, parameterName=Strips}
  n4 = Sign(None=Multiply(A=@n2, B=@n3))
  Ceil(None=@n4)
  Divide(A=@n2, B=@n3)
  n5 = ComponentMask(None=WorldPosition().XYZ) {g=true, r=true}
  n6 = ComponentMask(None=VectorParameter() {parameterName=Offset, defaultValue=<rgba 0 0 0 1>}.RGB) {g=true, r=true}
  n7 = Add(A=Divide(A=Add(A=@n5, B=@n6), B=ScalarParameter() {defaultValue=10000, parameterName=WorldSize}), B=Constant2Vector() {g=0.5, r=0.5})
  Multiply(A=Clamp(None=TextureSampleParameter2D(UVs=@n7) {parameterName=RenderTarget}.RGB), B=Constant() {r=0.5})
  Rotator(Coordinate=@n1, Time=Constant() {r=3.14})
  Saturate(None=@n4)
  VectorParameter() {parameterName=Color, defaultValue=<rgba 1 0 0 1>}
 ]
}
